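# Niagara idle effect for the Funky Feather cap backpack cosmetic, emitting sparkles, a flare, and reactive mesh music notes and coins from sockets.
# DamageReactivity drives coin and note spawn rates, GlowReactivity drives inner sparkles and flare alpha, while outer sparkles thin out as User.Reactivity rises.
# Note colors follow the system's variant value and note velocity switches for front end.
# Sparkles and flare are tinted by time-of-day color.
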NiagaraSystem /BRCosmetics/Accessories/FORT_Backpacks/Backpack_M_MED_FunkyFeather_Cap/FX/NS_Backpack_FunkyFeather_Cap_Idle
dependencies {
 usedDataInterfaces [
  &NiagaraDataInterfaceArrayColor
  &NiagaraDataInterfaceMeshRendererInfo
  &NiagaraDataInterfaceVectorCurve
  &NiagaraDataInterfaceCurve
  &NiagaraDataInterfaceArrayFloat3
  &NiagaraDataInterfaceSocketReader
 ]
 usedDynamicInputs [
  &FNTimeOfDayColor
  &MakeCustomFloatFromBool
  &MakeIntFromFloat
 ]
 usedModules [
  &self:SystemSpawnScript.SystemScriptSource.SystemScriptGraph.NiagaraNodeAssignment_1.SetVariables_#1
  &SystemState
  &self:SystemSpawnScript.SystemScriptSource.SystemScriptGraph.NiagaraNodeAssignment_9.SetVariables_#2
  &self:SystemSpawnScript.SystemScriptSource.SystemScriptGraph.NiagaraNodeAssignment_0.SetVariables_#3
  &self:SystemSpawnScript.SystemScriptSource.SystemScriptGraph.NiagaraNodeAssignment_2.SetVariables_#4
 ]
}
emitters {
 Coins {
  emitterUpdateScript [
   &EmitterState(Life Cycle Mode=::NewEnumerator1, Loop Duration=1, UseLoopDelay=false)
   &SpawnRate(SpawnRate=&Multiply_Float(A=200, B=$System.DamageReactivity), Use Spawn Probability=false)
  ]
  particleSpawnScript [
   &InitializeParticle(
    Lifetime=&RandomRangeFloat(Minimum=0.5, Maximum=0.8)
    Mesh Scale Mode=::NewEnumerator3
    UsePositionOffset=true
    Write Material Random=true
   )
   &AddVelocity(
    Cone Angle=40
    Cone Axis=&RandomRangeVector(Minimum=<-1 -1 0.2472>, Maximum=<1 -1 0.5>)
    Module.UseConeFalloff=false
    Module.UseVelDistribution=false
    Override Seed=false
    Velocity Mode=::NewEnumerator2
    Velocity Speed=180
   )
   &SocketLocation(
    SamplePosOffset=true
    Sampled Position Offset=&RandomRangeVector(Minimum=<0 -5 0>, Maximum=<0 5 0>)
    Socket Reader=$System.Socket Reader
    Socket Sampling Mode=::NewEnumerator3
    VelocitySpeedLimit=false
   )
   &InitialMeshOrientation(Rotation Coordinate Space=::NewEnumerator2, Use Rotation Vector=true)
   /SetVariables(Particles.MeshIndex=&RandomRangeInt(Minimum=0, Maximum=3)) {isSetParameters=true, module=SetVariables_#5}
  ]
  particleUpdateScript [
   &ParticleState()
   &ScaleMeshSize(Scale Factor=&VectorFromFloat(Value=&FloatFromCurve(FloatCurve=DataInterface() {propertyValues {Curve {
         defaultValue=0
         keys [
          {arriveTangent=22.5426, interpMode=RCIM_Cubic, leaveTangent=22.5426, tangentMode=RCTM_User, time=0, value=0}
          {arriveTangent=0.3041, interpMode=RCIM_Cubic, leaveTangent=0.3041, tangentMode=RCTM_User, time=0.0719, value=0.9847}
          {arriveTangent=-1.061, interpMode=RCIM_Cubic, leaveTangent=-1.061, time=0.7473, value=0.9144}
          {arriveTangent=-0.0405, interpMode=RCIM_Cubic, leaveTangent=-0.0405, tangentMode=RCTM_User, time=1, value=0}
         ]
        }}})))
   &Drag(Use Linear Drag=true, Use Rotational Drag=true)
   &InheritSourceMovement(
    Applied Position Scale=&VectorFromFloat(Value=&FloatFromCurve(FloatCurve=DataInterface() {propertyValues {Curve {keys [<at 0 1>,<at 0.3 0.2>]}}}))
    ApplyForce=false
    ApplyPosition=true
    ApplyPositionFalloff=false
   )
   &UpdateMeshOrientation(
    Rotation Rate=&FloatFromCurve(
     FloatCurve=DataInterface() {propertyValues {Curve {
        defaultValue=1
        keys [
         {arriveTangent=1.3423, interpMode=RCIM_Cubic, time=0.0699, value=0}
         {arriveTangent=6.1004, interpMode=RCIM_Cubic, leaveTangent=6.1004, tangentMode=RCTM_User, time=0.1149, value=0.4949}
         {arriveTangent=0.4937, interpMode=RCIM_Cubic, leaveTangent=0.4937, tangentMode=RCTM_User, time=0.318, value=1}
        ]
       }}}
     Scale Curve=&RandomRangeFloat(Minimum=1.5, Maximum=2.5)
    )
    Rotation Vector=&RandomVector(Vector Scale=1)
   )
   &GravityForce(Gravity=<0 0 -300>)
   &SolveForcesAndVelocity(Clamp Velocity=false, Limit Acceleration=false)
  ]
  rendererClasses [&NiagaraMeshRendererProperties]
  renderers [{rendererClass=&NiagaraMeshRendererProperties, rendererIndex=0}]
 }
 Flare {
  emitterUpdateScript [
   &EmitterState(Life Cycle Mode=::NewEnumerator1, Loop Behavior=::NewEnumerator1, Loop Duration Mode=::NewEnumerator1, UseLoopDelay=false)
   &SpawnBurst_Instantaneous(Use Spawn Probability=false, UseLoopCountLimit=true)
  ]
  particleSpawnScript [
   &PartitionParticles(Number Of Partitions=3)
   &InitializeParticle(
    Color=<rgba 3 1.7504 0.1784 0.5478>
    Lifetime=0.5
    Sprite Size Mode=::NewEnumerator3
    Sprite UV Mode=::NewEnumerator2
    Uniform Sprite Size=50
    UsePositionOffset=false
    Write Material Random=true
   )
   &SocketLocation(SamplePosOffset=false, Socket Reader=$System.Socket Reader, Socket Sampling Mode=::NewEnumerator3, VelocitySpeedLimit=false)
   &CameraOffset()
  ]
  particleUpdateScript [
   /SetVariables(Particles.ReactLifetime:NiagaraFloat=$Particles.NormalizedAge) {isSetParameters=true}
   /SetVariables(Particles.ReactLifetime:NiagaraFloat=$Particles.NormalizedAge) {isSetParameters=true}
   &ParticleState(
    Kill Particles When Lifetime Has Elapsed=false
    Let Infinitely Lived Particles Die When Emitter Deactivates=true
    Loop Particles Lifetime=true
   )
   &ScaleColor(
    Scale Alpha=$System.GlowReactivity
    Scale RGB=&MakeVectorFromLinearColorRGB(Color=$System.TOD Color)
    ScaleA=true
    ScaleRGB=true
   )
  ]
  rendererClasses [&NiagaraSpriteRendererProperties]
  renderers [{rendererClass=&NiagaraSpriteRendererProperties}]
 }
 Notes {
  emitterUpdateScript [
   &EmitterState(Life Cycle Mode=::NewEnumerator1, Loop Duration=1, UseLoopDelay=false)
   &SpawnRate(SpawnRate=&Multiply_Float(A=40, B=$System.DamageReactivity), Use Spawn Probability=false)
  ]
  particleSpawnScript [
   /SetVariables(Particles.NoteColor:LinearColor=&MakeCustomLinearColorFromBool(
     Bool=&SetBoolByIntComparison(A=$System.intVariant, B=0, Comparison Type=::NewEnumerator4)
     True Color=&SelectLinearColorFromArray(
      Array Sampling Mode=::NewEnumerator0
      Direct Array Index=?
      Color Selection Array=DataInterface() {propertyValues {ColorData [<rgba 0.4102 0.5435 0.4517 1>, <rgba 0.5473 0.3548 0.422 1>]}}
      Random Seed=0
     )
     False Color=&MakeCustomLinearColorFromBool(
      Bool=&SetBoolByIntComparison(A=$System.intVariant, B=1, Comparison Type=::NewEnumerator4)
      True Color=&SelectLinearColorFromArray(
       Array Sampling Mode=::NewEnumerator0
       Direct Array Index=?
       Color Selection Array=DataInterface() {propertyValues {ColorData [<rgba 0.5782 0.5359 0.2689 1>, <rgba 0.2184 0.6055 0.652 1>]}}
       Random Seed=0
      )
      False Color=&SelectLinearColorFromArray(
       Array Sampling Mode=::NewEnumerator0
       Direct Array Index=?
       Color Selection Array=DataInterface() {propertyValues {ColorData [<rgba 0.2983 0.3136 0.3923 1>, <rgba 0.5782 0.3923 0.31 1>]}}
       Random Seed=0
      )
     )
    )) {isSetParameters=true}
   &InitializeParticle(
    Color=$Particles.NoteColor
    Lifetime=&RandomRangeFloat(Minimum=0.6, Maximum=1)
    Mesh Scale Mode=::NewEnumerator3
    UsePositionOffset=true
    Write Material Random=true
   )
   &AddVelocity(
    Velocity=&FN_FrontEndVector3Modifier(
     Gameplay=&RandomRangeVector(Minimum=<-60 -60 -5>, Maximum=<60 -120 -15>)
     Front End=&RandomRangeVector(Minimum=<-60 -60 10>, Maximum=<60 -120 -30>)
    )
    Velocity Speed Scale=3
   )
   &SocketLocation(
    SamplePosOffset=true
    Sampled Position Offset=&RandomRangeVector(Minimum=<0 -5 0>, Maximum=<0 5 0>)
    Socket Reader=$System.Socket Reader
    Socket Sampling Mode=::NewEnumerator3
    VelocitySpeedLimit=false
   )
   &InitialMeshOrientation(
    Mesh Orientation Mode=::NewEnumerator6
    Orientation Quaternion:Quat4f=$Output.SocketLocation.SampledRotation
    Rotation=&RandomRangeVector(Minimum=<0 0 0>, Maximum=<0 0 1>)
    Rotation Coordinate Space=::NewEnumerator2
    Use Rotation Vector=true
   )
   /SetVariables(Particles.MeshIndex=&RandomRangeInt(Minimum=0, Maximum=3)) {isSetParameters=true, module=SetVariables_#5}
  ]
  particleUpdateScript [
   &ParticleState()
   &ScaleMeshSize(Scale Factor=&Multiply_VectorByFloat(
     Vector=&VectorFromCurve(
      VectorCurve=DataInterface() {propertyValues {XCurve {keys [<at 0 1>,<at 1 1>]}, YCurve {keys [<at 0 1>,<at 1 1>]}, ZCurve {keys [<at 0 0.1>,<at 0.4 1>]}}}
      CurveIndex=$Particles.Age
     )
     Float=&FloatFromCurve(FloatCurve=DataInterface() {propertyValues {Curve {
         defaultValue=0
         keys [
          {arriveTangent=22.5426, interpMode=RCIM_Cubic, leaveTangent=22.5426, tangentMode=RCTM_User, time=0, value=0}
          {arriveTangent=0.3041, interpMode=RCIM_Cubic, leaveTangent=0.3041, tangentMode=RCTM_User, time=0.0719, value=0.9847}
          {arriveTangent=-1.061, interpMode=RCIM_Cubic, leaveTangent=-1.061, time=0.7473, value=0.9144}
          {arriveTangent=-0.0405, interpMode=RCIM_Cubic, leaveTangent=-0.0405, tangentMode=RCTM_User, time=1, value=0}
         ]
        }}})
    ))
   &Drag(Drag=3, Use Linear Drag=true, Use Rotational Drag=true)
   &CurlNoiseForce(
    Bias Noise=true
    Bias Noise Field:Vector3f=&RandomVector(Vector Scale=1)
    Noise Frequency=1
    Noise Quality / Cost=::NewEnumerator0
    Noise Strength=&RandomRangeFloat(Minimum=12, Maximum=60)
    Pan Noise=true
    Pan Noise Field=<0 0 0.2>
   )
   &InheritSourceMovement(
    Applied Position Scale=&VectorFromFloat(Value=&FloatFromCurve(FloatCurve=DataInterface() {propertyValues {Curve {keys [<at 0 1>,<at 0.3 0.2>]}}}))
    ApplyForce=false
    ApplyPosition=true
    ApplyPositionFalloff=false
   )
   &UpdateMeshOrientation(
    Rotation Rate=&FloatFromCurve(
     FloatCurve=DataInterface() {propertyValues {Curve {
        defaultValue=1
        keys [
         {arriveTangent=1.3423, interpMode=RCIM_Cubic, time=0.0699, value=0}
         {arriveTangent=6.1004, interpMode=RCIM_Cubic, leaveTangent=6.1004, tangentMode=RCTM_User, time=0.1149, value=0.4949}
         {arriveTangent=0.4937, interpMode=RCIM_Cubic, leaveTangent=0.4937, tangentMode=RCTM_User, time=0.318, value=1}
        ]
       }}}
     Scale Curve=&RandomRangeFloat(Minimum=0.8, Maximum=1.6)
    )
    Rotation Vector=&RandomVector(Vector Scale=1)
   )
   &SolveForcesAndVelocity(Clamp Velocity=false, Limit Acceleration=false)
  ]
  rendererClasses [&NiagaraMeshRendererProperties]
  renderers [{rendererClass=&NiagaraMeshRendererProperties, rendererIndex=0}]
 }
 Sparkle {
  emitterUpdateScript [
   &EmitterState(Life Cycle Mode=::NewEnumerator1, Loop Duration=0.5, UseLoopDelay=false)
   &SpawnRate(
    SpawnRate=&Multiply_Float(A=20, B=&OneMinusFloat(Float=$User.Reactivity))
    Use Spawn Probability=false
   )
  ]
  particleSpawnScript [
   &PartitionParticles(Number Of Partitions=3)
   &InitializeParticle(
    Color=<rgba 0.8448 0.9614 1 0.35>
    Lifetime=&RandomRangeFloat(Minimum=0.4, Maximum=0.6)
    Sprite Size Mode=::NewEnumerator3
    Sprite UV Mode=::NewEnumerator2
    Uniform Sprite Size=&FloatFromCurve(
     FloatCurve=DataInterface() {propertyValues {Curve {keys [<at 0 3 RCIM_Cubic>, {arriveTangent=3.5447, interpMode=RCIM_Cubic, leaveTangent=3.5447, tangentMode=RCTM_User, time=1, value=5}]}}}
     CurveIndex=&RandomRangeFloat(Minimum=0, Maximum=1)
    )
    UsePositionOffset=false
    Write Material Random=true
   )
   &SocketLocation(
    SamplePosOffset=false
    Socket Index=$Particles.PartitionParticles.Partition
    Socket Reader=$System.Socket Reader
    Socket Sampling Mode=::NewEnumerator3
    VelocitySpeedLimit=false
   )
   &ShapeLocation(
    Additional Quaternion Rotation=false
    MaskSpawnGroup=false
    Non Uniform Scale=&Multiply_VectorByFloat(
     Vector=&SelectVectorFromArray(
      Direct Array Index=$Particles.PartitionParticles.Partition
      Vector Selection Array=DataInterface() {propertyValues {FloatData [<0 0 0>, <0 0 0>, <6 6 6>]}}
     )
     Float=0.85
    )
    Override Seed=false
    Rotation Mode=::NewEnumerator3
    Rotation Quaternion=$Output.SocketLocation.SampledRotation
    Sphere Radius=1
    Sphere Surface Distribution=1
    UseOwnerScale=true
   ) {module=ShapeLocation001}
   &ShapeLocation(
    Additional Quaternion Rotation=false
    Hemisphere Distribution=<1 0.5>
    MaskSpawnGroup=false
    Non Uniform Scale=&Multiply_VectorByFloat(
     Vector=&SelectVectorFromArray(
      Direct Array Index=$Particles.PartitionParticles.Partition
      Vector Selection Array=DataInterface() {propertyValues {FloatData [<10 13 11>, <10 13 6>, <0 0 0>]}}
     )
     Float=0.85
    )
    Override Seed=false
    Rotation Angle Type:ENiagara_AngleInput=::NewEnumerator0
    Rotation Mode=::NewEnumerator3
    Rotation Quaternion=$Output.SocketLocation.SampledRotation
    Sphere Radius=1
    Sphere Surface Distribution=1
    UseOwnerScale=true
   )
   &CameraOffset(Camera Offset Amount=10)
  ]
  particleUpdateScript [
   &ParticleState()
   &ScaleSpriteSize(Scale Factor=&Vector2DFromFloat(Value=&FloatFromCurve(
      FloatCurve=DataInterface() {propertyValues {Curve {keys [
          <at 0 0 RCIM_Cubic>
          {arriveTangent=-1.0005, interpMode=RCIM_Cubic, leaveTangent=-1.0005, tangentMode=RCTM_User, time=0.2894, value=1}
          {arriveTangent=0.1497, interpMode=RCIM_Cubic, leaveTangent=0.1497, tangentMode=RCTM_User, time=0.5815, value=0.4733}
          {arriveTangent=-0.8218, tangentMode=RCTM_User, time=1.0011, value=0.002}
         ]}}}
      Scale Curve=&FloatFromCurve(
       FloatCurve=DataInterface() {propertyValues {Curve {keys [<at 0 3>,<at 0.2 1.2>,<at 1 1>]}}}
       CurveIndex=&RandomRangeFloat(Minimum=0, Maximum=1, Evaluation Type=::NewEnumerator1)
      )
     )))
   &ScaleColor(
    Scale Alpha=&FloatFromCurve(FloatCurve=DataInterface() {propertyValues {Curve {keys [<at 0.5 1 RCIM_Cubic>,<at 1 0 RCIM_Cubic>]}}})
    Scale RGB=&MakeVectorFromLinearColorRGB(Color=$System.TOD Color)
    ScaleA=true
    ScaleRGB=true
   )
   &DynamicMaterialParameters(
    Index 0 Param 1=3
    Index 0 Param 2=1
    Index 0 Param 3=0.5
    Module.Index0Param0Write=true
    Module.Index0Param1Write=true
    Module.Index0Param2Write=true
    Module.Index0Param3Write=true
   )
  ]
  rendererClasses [&NiagaraSpriteRendererProperties]
  renderers [{rendererClass=&NiagaraSpriteRendererProperties}]
 }
 Sparkle_Inner {
  emitterUpdateScript [
   &EmitterState(Life Cycle Mode=::NewEnumerator1, Loop Duration=0.5, UseLoopDelay=false)
   &SpawnRate(SpawnRate=&Multiply_Float(A=80, B=$System.GlowReactivity), Use Spawn Probability=false)
  ]
  particleSpawnScript [
   &PartitionParticles(Number Of Partitions=3)
   &InitializeParticle(
    Color=<rgba 1 0.4935 0.0584 0.8087>
    Lifetime=&RandomRangeFloat(Minimum=0.4, Maximum=0.6)
    Sprite Size Mode=::NewEnumerator3
    Sprite UV Mode=::NewEnumerator2
    Uniform Sprite Size=&FloatFromCurve(
     FloatCurve=DataInterface() {propertyValues {Curve {keys [<at 0 3 RCIM_Cubic>, {arriveTangent=3.5447, interpMode=RCIM_Cubic, leaveTangent=3.5447, tangentMode=RCTM_User, time=1, value=5}]}}}
     CurveIndex=&RandomRangeFloat(Minimum=0, Maximum=1)
     Scale Curve=2
    )
    UsePositionOffset=false
    Write Material Random=true
   )
   &SocketLocation(SamplePosOffset=false, Socket Reader=$System.Socket Reader, Socket Sampling Mode=::NewEnumerator3, VelocitySpeedLimit=false)
   &ShapeLocation(
    Additional Quaternion Rotation=false
    MaskSpawnGroup=false
    Non Uniform Scale=<7 8 2>
    Override Seed=false
    Rotation Mode=::NewEnumerator3
    Rotation Quaternion=$Output.SocketLocation.SampledRotation
    Sphere Radius=1
    Sphere Surface Distribution=1
    UseOwnerScale=true
   ) {module=ShapeLocation001}
   &AddVelocity(Velocity=&Add_Vector(A=&RandomVector(Vector Scale=2), B=<0 0 5>))
   &CameraOffset(Camera Offset Amount=3)
  ]
  particleUpdateScript [
   &ParticleState()
   &SolveForcesAndVelocity(Clamp Velocity=false, Limit Acceleration=false)
   &ScaleSpriteSize(Scale Factor=&Vector2DFromFloat(Value=&FloatFromCurve(
      FloatCurve=DataInterface() {propertyValues {Curve {keys [
          <at 0 0 RCIM_Cubic>
          {arriveTangent=-1.0005, interpMode=RCIM_Cubic, leaveTangent=-1.0005, tangentMode=RCTM_User, time=0.2894, value=1}
          {arriveTangent=0.1497, interpMode=RCIM_Cubic, leaveTangent=0.1497, tangentMode=RCTM_User, time=0.5815, value=0.4733}
          {arriveTangent=-0.8218, tangentMode=RCTM_User, time=1.0011, value=0.002}
         ]}}}
      Scale Curve=&FloatFromCurve(
       FloatCurve=DataInterface() {propertyValues {Curve {keys [<at 0 3>,<at 0.2 1.2>,<at 1 1>]}}}
       CurveIndex=&RandomRangeFloat(Minimum=0, Maximum=1, Evaluation Type=::NewEnumerator1)
      )
     )))
   &ScaleColor(
    Scale Alpha=&FloatFromCurve(FloatCurve=DataInterface() {propertyValues {Curve {keys [<at 0.5 1 RCIM_Cubic>,<at 1 0 RCIM_Cubic>]}}})
    Scale RGB=&MakeVectorFromLinearColorRGB(Color=$System.TOD Color)
    ScaleA=true
    ScaleRGB=true
   )
   &DynamicMaterialParameters(
    Index 0 Param 1=3
    Index 0 Param 2=1
    Index 0 Param 3=0.5
    Module.Index0Param0Write=true
    Module.Index0Param1Write=true
    Module.Index0Param2Write=true
    Module.Index0Param3Write=true
   )
  ]
  rendererClasses [&NiagaraSpriteRendererProperties]
  renderers [{rendererClass=&NiagaraSpriteRendererProperties}]
 }
}
summary {
 systemName=NS_Backpack_FunkyFeather_Cap_Idle
 emitters [
  {emitterName=Notes, rendererClasses [&NiagaraMeshRendererProperties]}
  {emitterName=Sparkle, rendererClasses [&NiagaraSpriteRendererProperties]}
  {emitterName=Sparkle_Inner, rendererClasses [&NiagaraSpriteRendererProperties]}
  {emitterName=Flare, rendererClasses [&NiagaraSpriteRendererProperties]}
  {emitterName=Coins, rendererClasses [&NiagaraMeshRendererProperties]}
 ]
 userVariables [
  User.bIsFrontEnd {type=NiagaraBool}
  User.bIsFrontEndPreview {type=NiagaraBool}
  User.DebugReactivity {type=NiagaraBool}
  User.GlowVal {type=NiagaraFloat, defaultValue=0, description=""}
  User.ReactiveTriggered {type=NiagaraBool}
  User.Reactivity {type=NiagaraFloat, defaultValue=0, description=""}
  User.Variant {type=NiagaraFloat, defaultValue=0, description=""}
 ]
}
