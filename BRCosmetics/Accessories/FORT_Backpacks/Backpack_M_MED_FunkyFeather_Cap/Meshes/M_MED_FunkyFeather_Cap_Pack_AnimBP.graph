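# Anim Blueprint for the FunkyFeather Cap back bling, animating its duck, head, money, glasses and watch bones.
# Additive duck/head offsets feed AnimDynamics springs (glasses gated by not-DBNO), then a cached pose.
# An int blend zero-scales either the watch or third money bone, toggled after timed pulse sequences on damage or in the frontend.
# The EventGraph also derives backpack offsets from spine yaw, with no visible AnimGraph use.
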AnimBlueprint /BRCosmetics/Accessories/FORT_Backpacks/Backpack_M_MED_FunkyFeather_Cap/Meshes/M_MED_FunkyFeather_Cap_Pack_AnimBP
target_skeleton=&/M_MED_FunkyFeather_Cap_Pack_Skeleton
graphs [
 AnimGraph {nodes [
   n1 = Use cached pose 'SavedPose25345'()
   n2 = "Transform (Modify) Bone - Bone: dyn_watch_base"(
    ComponentPose="Transform (Modify) Bone - Bone: dyn_money_3_base"(ComponentPose=Local To Component(LocalPose=@n1), Alpha=1, Translation="0,0,0", Rotation="0,0,0", Scale="0,0,0") {scaleMode=BMM_Additive, boneToModify {boneName=dyn_money_3_base}, scale=<0 0 0>}
    Alpha=1
    Translation="0,0,0"
    Rotation="0,0,0"
    Scale="1,1,1"
   ) {scaleMode=BMM_Additive, boneToModify {boneName=dyn_watch_base}}
   n3 = Use cached pose 'SavedPose25345'()
   n4 = "Transform (Modify) Bone - Bone: dyn_money_3_base"(
    ComponentPose="Transform (Modify) Bone - Bone: dyn_watch_base"(ComponentPose=Local To Component(LocalPose=@n3), Alpha=1, Translation="0,0,0", Rotation="0,0,0", Scale="0,0,0") {scaleMode=BMM_Additive, boneToModify {boneName=dyn_watch_base}, scale=<0 0 0>}
    Alpha=1
    Translation="0,0,0"
    Rotation="0,0,0"
    Scale="1,1,1"
   ) {scaleMode=BMM_Additive, boneToModify {boneName=dyn_money_3_base}}
   Output Pose(Result=Blend Poses by int!(
     BlendPose_0:Pose Link Structure=Component To Local(ComponentPose=@n2)
     BlendPose_1:Pose Link Structure=Component To Local(ComponentPose=@n4)
     BlendTime_0:float=0.1
     BlendTime_1:float=0.1
     ActiveChildIndex:int="Get select "() -> ("select ":int=0)
    ) -> (Pose:Pose Link Structure) {is=title AnimGraphNode_BlendListByInt, blendPose [{linkId=-1, sourceLinkId=-1}, {linkId=-1, sourceLinkId=-1}], blendTime [0.1,0.1]}) {layerGroup=None}
   n5 = "Transform (Modify) Bone - Bone: dyn_duck_base"(ComponentPose=Local To Component(LocalPose=Input Pose()), Alpha=1, Translation="0,0,0", Rotation="0,-19,0", Scale="1,1,1") {
    rotationMode=BMM_Additive
    rotationSpace=BCS_BoneSpace
    translationMode=BMM_Additive
    translationSpace=BCS_BoneSpace
    alphaBoolBlend {blendOption=Linear}
    alphaScaleBiasClamp {bInterpResult=true, interpSpeedDecreasing=15, interpSpeedIncreasing=18}
    boneToModify {boneName=dyn_duck_base}
    rotation {yaw=-19}
   }
   n6 = "Transform (Modify) Bone - Bone: dyn_head_base"(ComponentPose=@n5, Alpha=1, Translation="0,0,0", Rotation="0,0,90", Scale="1,1,1") {
    rotationMode=BMM_Additive
    rotationSpace=BCS_BoneSpace
    translationMode=BMM_Additive
    translationSpace=BCS_BoneSpace
    alphaBoolBlend {blendOption=Linear}
    alphaScaleBiasClamp {bInterpResult=true, interpSpeedDecreasing=15, interpSpeedIncreasing=18}
    boneToModify {boneName=dyn_head_base}
    rotation {roll=90}
   }
   n7 = Anim Dynamics - Bone: dyn_duck_01(ComponentPose=@n6, Alpha=1, ExternalForce="0,0,0") {
    angularDampingOverride=1
    angularSpringConstant=1200
    bAngularSpring=true
    bOverrideAngularDamping=true
    bUseGravityOverride=true
    simulationSpace=RootRelative
    boundBone {boneName=dyn_duck_01}
    componentLinearAccScale {z=0.3}
    componentLinearVelScale {z=0.3}
    gravityOverride {z=-980}
    physicsBodyDefinitions [{
      boundBone {boneName=dyn_duck_01}
      boxExtents=<11 11 11>
      constraintSetup {angularLimitsMax=<0 0 0>, angularLimitsMin=<0 0 -5>, angularTarget=<1 0 0>}
      localJointOffset=<0 0 1>
     }]
    relativeSpaceBone {boneName=root}
    simSpaceSettings {maxAngularAcceleration=50, maxAngularVelocity=10, simSpaceAngularAlpha=1}
   }
   n8 = Anim Dynamics - Bone: dyn_money_01(ComponentPose=@n7, Alpha=1, ExternalForce="0,0,0") {
    angularDampingOverride=1
    angularSpringConstant=1200
    bAngularSpring=true
    bOverrideAngularDamping=true
    bUseGravityOverride=true
    simulationSpace=RootRelative
    boundBone {boneName=dyn_money_01}
    componentLinearAccScale {z=0.3}
    componentLinearVelScale {z=0.3}
    gravityOverride {z=-980}
    physicsBodyDefinitions [{
      boundBone {boneName=dyn_money_01}
      boxExtents=<11 11 11>
      constraintSetup {angularLimitsMax=<0 20 0>, angularLimitsMin=<0 -44 0>, angularTarget=<1 0 0>}
      localJointOffset=<0.6 0 0>
     }]
    relativeSpaceBone {boneName=root}
    simSpaceSettings {maxAngularAcceleration=50, maxAngularVelocity=10, simSpaceAngularAlpha=1}
   }
   n9 = Anim Dynamics - Bone: dyn_money_2_01(ComponentPose=@n8, Alpha=1, ExternalForce="0,0,0") {
    angularDampingOverride=1
    angularSpringConstant=1200
    bAngularSpring=true
    bOverrideAngularDamping=true
    bUseGravityOverride=true
    simulationSpace=RootRelative
    boundBone {boneName=dyn_money_2_01}
    componentLinearAccScale {z=0.3}
    componentLinearVelScale {z=0.3}
    gravityOverride {z=-980}
    physicsBodyDefinitions [{
      boundBone {boneName=dyn_money_2_01}
      boxExtents=<11 11 11>
      constraintSetup {angularLimitsMax=<0 20 0>, angularLimitsMin=<0 -44 0>, angularTarget=<1 0 0>}
      localJointOffset=<0.6 0 0>
     }]
    relativeSpaceBone {boneName=root}
    simSpaceSettings {maxAngularAcceleration=50, maxAngularVelocity=10, simSpaceAngularAlpha=1}
   }
   n10 = Anim Dynamics - Bone: dyn_glasses_01!(ComponentPose=@n9, bAlphaBoolEnabled:bool="Get Dbno "() -> ("Dbno ":bool), ExternalForce="0,0,0") -> (Pose) {
    alphaInputType=Bool
    angularDampingOverride=1
    bAngularSpring=true
    bOverrideAngularDamping=true
    bUseGravityOverride=true
    simulationSpace=RootRelative
    boundBone {boneName=dyn_glasses_01}
    componentLinearAccScale {z=0.3}
    componentLinearVelScale {z=0.3}
    gravityOverride {z=-980}
    physicsBodyDefinitions [{
      boundBone {boneName=dyn_glasses_01}
      boxExtents=<11 11 11>
      constraintSetup {angularLimitsMax=<0 0 0>, angularLimitsMin=<0 -15 0>, angularTarget=<1 0 0>}
      localJointOffset=<0 0 1>
     }]
    relativeSpaceBone {boneName=root}
    simSpaceSettings {maxAngularAcceleration=50, maxAngularVelocity=10, simSpaceAngularAlpha=1}
   }
   Save cached pose 'SavedPose25345'(Pose=Component To Local(ComponentPose=@n10)) {cacheName=SavedPose25345}
  ]}
 EventGraph {nodes [
   Event BlueprintUpdateAnimation()
   > n1 = Sequence!(execute) -> (then_0, then_1, then_2:Exec)
   > Branch(Condition=Get bIsInFrontEnd())
   > Delay(Duration=5)
   > n2 = Sequence!(execute) -> (then_0, then_1)
   > Delay(Duration=1)
   > Set bEnemyPlayerEliminated(bEnemyPlayerEliminated:double=1) -> (Output_Get:double)
   > Delay(Duration=0.2)
   > Set bEnemyPlayerEliminated(bEnemyPlayerEliminated=0.7)
   > Delay(Duration=0.2)
   > Set bEnemyPlayerEliminated(bEnemyPlayerEliminated=1)
   > Delay(Duration=0.2)
   > Set bEnemyPlayerEliminated(bEnemyPlayerEliminated=0.8)
   > Delay(Duration=0.2)
   > Set bEnemyPlayerEliminated(bEnemyPlayerEliminated=1)
   > Delay(Duration=0.2)
   > Set bEnemyPlayerEliminated(bEnemyPlayerEliminated=0.9)
   > Delay(Duration=0.1)
   > Set bEnemyPlayerEliminated(bEnemyPlayerEliminated=0)
   > Delay(Duration=0.5)
   > "Set select "("select ":int=Select(Option 0:int=1, Option 1:int=0, Index:int=Get option() -> (option:int=0)) -> (ReturnValue:int)) -> (Output_Get:int=0)
   > n3 = Branch(Condition="integer > integer"(A="Get select "() -> ("select "=0)))
   > Set option(option:int=1) -> (Output_Get:int=0)
   Set option(execute=@n3.else, option=0) -> (Output_Get=0)
   Delay(execute=@n2.then_1, Duration=1)
   > Set bEnemyPlayerEliminated_1(bEnemyPlayerEliminated_1:double=1) -> (Output_Get:double)
   > Delay(Duration=1.1)
   > Set bEnemyPlayerEliminated_1(bEnemyPlayerEliminated_1=0)
   "Set Dbno "(execute=@n1.then_2, "Dbno ":bool=NOT Boolean(A=Get bIsDBNO())) -> (Output_Get:bool)
   n4 = GetOwnerSkeletalMeshComponent()
   n5 = BreakRotator(InRot=BreakTransform(InTransform=GetDeltaTransformFromRefPose(self=@n4, BoneName=spine_05, BaseName=spine_01)).Rotation)
   n6 = MapRangeClamped(Value=@n5.Yaw, InRangeA=-10, InRangeB=-50, OutRangeB=35)
   Set backpack_rotate(execute=@n1.then_1, backpack_rotate=MakeRotator(Roll=@n6, Pitch=float / float(A=@n6, B=-10)))
   > Set backpack_translation(backpack_translation=MakeVector(Y=MapRangeClamped(Value=@n5.Yaw, InRangeA=-15, InRangeB=-50, OutRangeB=-15), Z=-6.5))
   n7 = Event BlueprintProcessAnimInputEvent()
   > Branch(Condition=OR Boolean!(
     A=IsNamed(self=@n7.AnimInputEvent, InEventName=aie.PlayerVitality.ShieldsDecreased)
     B=IsNamed(self=@n7.AnimInputEvent, InEventName=aie.PlayerVitality.HealthDecreased)
     C:bool
    ) -> (ReturnValue))
   > n8 = Sequence()
   > Set bEnemyPlayerEliminated(bEnemyPlayerEliminated=1)
   > Delay(Duration=0.2)
   > Set bEnemyPlayerEliminated(bEnemyPlayerEliminated=0.7)
   > Delay(Duration=0.2)
   > Set bEnemyPlayerEliminated(bEnemyPlayerEliminated=1)
   > Delay(Duration=0.2)
   > Set bEnemyPlayerEliminated(bEnemyPlayerEliminated=0.8)
   > Delay(Duration=0.2)
   > Set bEnemyPlayerEliminated(bEnemyPlayerEliminated=1)
   > Delay(Duration=0.2)
   > Set bEnemyPlayerEliminated(bEnemyPlayerEliminated=0.9)
   > Delay(Duration=0.1)
   > Set bEnemyPlayerEliminated(bEnemyPlayerEliminated=0)
   > Delay(Duration=0.5)
   > "Set select "("select "=Select(Option 0=1, Option 1=0, Index=Get option() -> (option=0))) -> (Output_Get=0)
   > n9 = Branch(Condition="integer > integer"(A="Get select "() -> ("select "=0)))
   > Set option(option=1) -> (Output_Get=0)
   Set option(execute=@n9.else, option=0) -> (Output_Get=0)
   Set bEnemyPlayerEliminated_1(execute=@n8.then_1, bEnemyPlayerEliminated_1=1)
   > Delay(Duration=1.1)
   > Set bEnemyPlayerEliminated_1(bEnemyPlayerEliminated_1=0)
   TryGetPawnOwner()
  ]}
 NewEventDispatcher {nodes [NewEventDispatcher() {is=title K2Node_FunctionEntry}]}
]
variables [
 Active_Damage {type=boolean}
 backpack_rotate {type=Rotator}
 backpack_translation {type=Vector}
 bEnemyPlayerEliminated {type=number}
 bEnemyPlayerEliminated_1 {type=number}
 Clap Alpha {type=number}
 Clap_Timer {type=number}
 Clap_Timer_Length {type=number}
 Crank Count {type=number}
 DamageCount {type=number}
 "Dbno " {type=boolean}
 eliminated {type=boolean}
 Jaw Laugh Rot {type=Rotator}
 Jaw Laugh Rot 2 {type=Rotator}
 killcount {type=integer}
 Last Damage Count {type=number}
 NewEventDispatcher {default=()}
 option {type=integer}
 "select " {type=integer}
]
